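# Level Sequence for the Pistol Canary tentacle creature's circular beam attack animation.
# The skeletal mesh actor holds a fixed transform and carries a skeletal animation track plus a Control Rig track.
# Two static mesh actors, named as position controls, carry densely keyed transforms that swing through full-circle yaw sweeps, ending near their starting poses.
# The sequence shows no link from those controls to the rig, and no beam effect, audio or event tracks.
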LevelSequence /PistolCanary/Anim/SingleTentacle/Section02/PistolCanary_Attack_CircularBeam/PistolCanary_TentacleAttack_CircularBeam_anm
bindings [
 _1_Position_CTRL {
  id=#1
  possessed_class=FortStaticMeshActor
  tracks [
   {
    type=MovieScene3DTransformTrack
    sections [{
      type=MovieScene3DTransformSection
      channels [
       Weight {type=MovieSceneScriptingFloatChannel, default=1}
       Location.X {
        type=MovieSceneScriptingDoubleChannel
        default=14575.0915
        thinned_from=433
        keys [[3600,14287.208],[3657,14393.6761],[3667,14185.6075],[3906,13854.2128],[3976,15320.0492],[3986,15254.9587],[4010,14345.6963],[4032,14287.208]]
       }
       Location.Y {type=MovieSceneScriptingDoubleChannel, default=72029.0982, thinned_from=433, keys [[3600,70618.9219],[4032,70618.9219]]}
       Location.Z {
        type=MovieSceneScriptingDoubleChannel
        default=4250.7427
        thinned_from=433
        keys [[3600,3947.8015],[3618,3934.241],[3620,3918.0092],[3622,3881.2371],[3624,3829.9878],[3625,3797.4788],[3627,3707.7771],[3630,3554.1292],[3631,3514.8353],[3632,3490.0026],[3643,3328.7449],[3648,3290.4908],[3651,3288.333],[3656,3328.076],[3661,3408.1948],[3665,3459.6648],[3668,3519.8215],[3670,3543.8123],[3673,3549.6076],[3679,3527.6627],[3932,3879.7717],[3972,3572.8123],[3979,3568.5976],[3982,3590.8603],[3998,3889.7231],[4003,3938.5207],[4032,3947.8015]]
       }
       Rotation.X {
        type=MovieSceneScriptingDoubleChannel
        default=0
        thinned_from=433
        keys [[3600,0],[3601,0.0062],[3602,0.0236],[3603,0.0503],[3604,0.0842],[3605,0.1235],[3606,0.1663],[3607,0.2105],[3608,0.2543],[3609,0.3008],[3610,0.3572],[3611,0.4272],[3612,0.5148],[3613,0.6172],[3614,0.7315],[3615,0.8618],[3616,1.0124],[3617,1.1873],[3618,1.3906],[3619,1.9061],[3620,2.9381],[3621,4.3696],[3622,6.084],[3623,7.9658],[3624,9.9006],[3625,12.0869],[3626,14.677],[3628,20.3279],[3629,23.0181],[3630,25.371],[3631,27.2013],[3632,28.3234],[3637,31.7623],[3640,33.6034],[3642,34.6652],[3644,35.5533],[3646,36.2326],[3648,36.6678],[3649,36.7828],[3650,36.8235],[3651,36.7644],[3652,36.5979],[3654,36.0343],[3656,35.3158],[3657,34.6763],[3658,33.887],[3659,33.2697],[3663,31.1799],[3664,30.5769],[3665,29.8712],[3666,29.0273],[3668,27.0413],[3669,26.3468],[3670,25.8905],[3671,25.6368],[3672,25.55],[3673,25.5945],[3674,25.7347],[3677,26.373],[3678,26.5395],[3679,26.6235],[3680,26.5894],[3780,19.9845],[3812,16.6294],[3816,16.0902],[3819,15.5895],[3821,15.1745],[3824,14.4363],[3828,13.2987],[3836,10.8663],[3839,10.0227],[3841,9.5093],[3843,9.0449],[3845,8.6359],[3847,8.2873],[3848,8.1367],[3849,8.0554],[3850,8.0657],[3852,8.1717],[3853,8.1709],[3854,8.07],[3857,7.5908],[3860,7.167],[3862,6.9225],[3864,6.7117],[3866,6.5366],[3868,6.3978],[3870,6.2945],[3872,6.2253],[3874,6.1873],[3877,6.1806],[3882,6.2628],[3889,6.4178],[3892,6.4404],[3894,6.4265],[3896,6.3839],[3897,6.3396],[3898,6.2664],[3899,6.1669],[3900,6.0436],[3901,5.8993],[3902,5.737],[3903,5.5597],[3905,5.173],[3909,4.3606],[3910,4.1677],[3911,3.9848],[3912,3.8148],[3913,3.6601],[3914,3.5232],[3915,3.4229],[3916,3.3652],[3917,3.3327],[3918,3.3079],[3919,3.2737],[3921,3.1481],[3922,3.114],[3923,3.1129],[3924,3.1464],[3925,3.2158],[3926,3.322],[3927,3.4655],[3928,3.6463],[3929,3.8644],[3930,4.1191],[3931,4.4099],[3932,4.7355],[3933,5.0948],[3934,5.4864],[3935,5.9085],[3936,6.3595],[3937,6.8372],[3938,7.3397],[3939,7.8648],[3940,8.4102],[3942,9.5528],[3952,15.6717],[3956,17.9953],[3959,19.6275],[3962,21.1313],[3964,22.0404],[3966,22.8557],[3968,23.5577],[3970,24.1252],[3972,24.5353],[3974,24.764],[3975,24.8027],[3976,24.7869],[3977,24.7136],[3978,24.58],[3979,24.3832],[3980,24.07],[3981,23.5922],[3982,22.9554],[3983,22.1668],[3984,21.2352],[3985,20.1714],[3986,18.9889],[3987,17.7036],[3988,16.334],[3989,14.9011],[3992,10.4571],[3993,9.0089],[3994,7.6167],[3995,6.3015],[3996,5.0811],[3997,3.9703],[3998,2.9798],[3999,2.1164],[4000,1.3829],[4001,0.7784],[4002,0.2987],[4003,-0.0637],[4004,-0.3185],[4005,-0.4779],[4006,-0.5555],[4007,-0.5661],[4008,-0.5252],[4009,-0.4482],[4010,-0.3504],[4011,-0.2463],[4012,-0.149],[4013,-0.0699],[4014,-0.0181],[4015,0],[4032,0]]
       }
       Rotation.Y {
        type=MovieSceneScriptingDoubleChannel
        default=0
        thinned_from=433
        keys [[3600,0],[3618,0],[3619,-0.076],[3620,-0.2856],[3621,-0.6007],[3622,-0.991],[3624,-1.8514],[3625,-2.313],[3626,-2.8481],[3627,-3.426],[3628,-4.0161],[3629,-4.5874],[3630,-5.1093],[3631,-5.5511],[3632,-5.8823],[3635,-6.6617],[3637,-7.1492],[3639,-7.5969],[3641,-7.9934],[3642,-8.1688],[3643,-8.3271],[3644,-8.4671],[3645,-8.5872],[3646,-8.6862],[3647,-8.7627],[3648,-8.8154],[3649,-8.8429],[3650,-8.8439],[3651,-8.7423],[3652,-8.4889],[3653,-8.1232],[3654,-7.6846],[3656,-6.7465],[3657,-6.1414],[3658,-5.2443],[3659,-4.0972],[3660,-2.7769],[3661,-1.3631],[3662,0.0643],[3663,1.4254],[3664,2.6405],[3665,3.6295],[3666,4.3128],[3667,4.6103],[3671,4.9031],[3673,5.0274],[3675,5.1288],[3677,5.2014],[3679,5.2392],[3680,5.2432],[3681,5.231],[3682,5.1996],[3683,5.1526],[3685,5.0269],[3687,4.8759],[3688,4.7818],[3689,4.6742],[3690,4.554],[3691,4.4223],[3692,4.2798],[3693,4.1274],[3694,3.9662],[3696,3.6208],[3698,3.2509],[3701,2.6659],[3705,1.8684],[3707,1.4786],[3708,1.2892],[3709,1.1046],[3710,0.9257],[3711,0.7535],[3712,0.5889],[3713,0.4328],[3714,0.2862],[3715,0.1498],[3716,0.0247],[3717,-0.0883],[3718,-0.1882],[3719,-0.2741],[3720,-0.3527],[3721,-0.4296],[3722,-0.5017],[3723,-0.5665],[3724,-0.6209],[3725,-0.6622],[3726,-0.695],[3727,-0.7262],[3728,-0.7556],[3729,-0.7834],[3731,-0.8345],[3733,-0.8796],[3735,-0.9193],[3737,-0.9539],[3739,-0.9838],[3742,-1.0207],[3745,-1.0492],[3749,-1.0764],[3754,-1.0971],[3773,-1.1304],[3776,-1.1409],[3777,-1.1416],[3778,-1.1362],[3779,-1.1249],[3780,-1.1085],[3781,-1.0873],[3782,-1.062],[3783,-1.033],[3784,-1.001],[3785,-0.9663],[3787,-0.8913],[3791,-0.7338],[3792,-0.696],[3793,-0.6599],[3794,-0.626],[3795,-0.5949],[3796,-0.5672],[3797,-0.5433],[3798,-0.5239],[3799,-0.5096],[3800,-0.5008],[3801,-0.4976],[3802,-0.4997],[3803,-0.5066],[3804,-0.5179],[3805,-0.5333],[3806,-0.5522],[3807,-0.5742],[3808,-0.5989],[3809,-0.6257],[3811,-0.6837],[3814,-0.773],[3815,-0.8008],[3816,-0.8264],[3817,-0.8491],[3819,-0.8875],[3820,-0.9111],[3821,-0.9378],[3824,-1.0249],[3825,-1.0525],[3826,-1.0776],[3827,-1.0992],[3828,-1.1165],[3829,-1.1286],[3830,-1.1348],[3831,-1.1346],[3832,-1.1273],[3833,-1.1126],[3834,-1.0903],[3835,-1.0602],[3836,-1.0224],[3837,-0.9768],[3838,-0.9239],[3839,-0.8639],[3840,-0.7973],[3841,-0.7247],[3842,-0.6469],[3843,-0.5645],[3844,-0.4786],[3845,-0.39],[3847,-0.2091],[3848,-0.1191],[3849,-0.0565],[3850,-0.0495],[3851,-0.0997],[3852,-0.2051],[3853,-0.3585],[3854,-0.5466],[3855,-0.7401],[3856,-0.9222],[3857,-1.0937],[3858,-1.2553],[3859,-1.4077],[3860,-1.5517],[3861,-1.6881],[3862,-1.8177],[3863,-1.9415],[3865,-2.1749],[3872,-2.937],[3874,-3.1667],[3876,-3.4127],[3878,-3.6796],[3880,-3.9715],[3882,-4.291],[3884,-4.6399],[3886,-5.019],[3888,-5.4275],[3890,-5.8638],[3893,-6.5642],[3902,-8.8263],[3904,-9.2869],[3906,-9.7082],[3909,-10.2665],[3914,-11.0726],[3915,-11.2158],[3916,-11.3309],[3917,-11.4194],[3918,-11.4806],[3919,-11.5115],[3920,-11.5069],[3921,-11.4572],[3922,-11.36],[3923,-11.2192],[3924,-11.039],[3925,-10.8233],[3926,-10.5762],[3927,-10.3014],[3928,-10.0028],[3930,-9.348],[3933,-8.272],[3938,-6.4176],[3940,-5.7078],[3941,-5.3669],[3942,-5.0371],[3943,-4.7197],[3944,-4.4155],[3945,-4.1254],[3946,-3.8501],[3947,-3.5901],[3948,-3.3457],[3949,-3.1173],[3950,-2.9047],[3951,-2.7075],[3952,-2.5257],[3953,-2.3611],[3954,-2.2152],[3955,-2.0896],[3956,-1.986],[3957,-1.9058],[3958,-1.8507],[3959,-1.8219],[3960,-1.8209],[3961,-1.8489],[3962,-1.907],[3963,-1.9961],[3964,-2.1169],[3965,-2.27],[3966,-2.4558],[3967,-2.6743],[3968,-2.9254],[3969,-3.2087],[3970,-3.5235],[3971,-3.8687],[3972,-4.2429],[3973,-4.6445],[3974,-5.0713],[3975,-5.5209],[3976,-5.9903],[3978,-6.9755],[3979,-7.4834],[3980,-8.0391],[3981,-8.6686],[3982,-9.3471],[3984,-10.7483],[3985,-11.4201],[3986,-12.0403],[3987,-12.5867],[3988,-13.0395],[3989,-13.382],[3990,-13.601],[3991,-13.6875],[3992,-13.6365],[3993,-13.4477],[3994,-13.1249],[3995,-12.6757],[3996,-12.1114],[3997,-11.4461],[3998,-10.696],[3999,-9.879],[4000,-9.0134],[4003,-6.3081],[4004,-5.4268],[4005,-4.5806],[4006,-3.7816],[4007,-3.0406],[4008,-2.3662],[4009,-1.7658],[4010,-1.2451],[4011,-0.8092],[4012,-0.4625],[4013,-0.209],[4014,-0.0532],[4015,0],[4032,0]]
       }
       Rotation.Z {
        type=MovieSceneScriptingDoubleChannel
        default=0
        thinned_from=433
        keys [[3600,0],[3618,0],[3619,0.045],[3620,0.1578],[3622,0.4512],[3623,0.5651],[3624,0.6149],[3625,0.6208],[3626,0.6234],[3627,0.6222],[3628,0.6177],[3629,0.6106],[3631,0.5937],[3633,0.5806],[3634,0.5712],[3635,0.5594],[3636,0.5454],[3637,0.5297],[3638,0.5125],[3640,0.475],[3644,0.3954],[3646,0.3575],[3647,0.3397],[3648,0.323],[3649,0.3077],[3650,0.2938],[3656,0.2118],[3657,0.1961],[3658,0.1879],[3659,0.1912],[3660,0.1973],[3661,0.2059],[3662,0.2161],[3665,0.2501],[3666,0.2601],[3667,0.268],[3668,0.2738],[3669,0.2783],[3670,0.2818],[3671,0.2843],[3673,0.2875],[3677,0.291],[3679,0.2943],[3680,0.297],[3681,0.3013],[3682,0.3079],[3683,0.3163],[3684,0.3265],[3685,0.3381],[3686,0.3508],[3687,0.3646],[3688,0.3795],[3689,0.3955],[3690,0.4125],[3691,0.4305],[3693,0.4691],[3695,0.5112],[3697,0.5563],[3700,0.6288],[3704,0.7329],[3720,1.1801],[3721,1.2118],[3722,1.2477],[3723,1.2891],[3724,1.3369],[3725,1.3924],[3749,2.8104],[3758,3.3948],[3766,3.9745],[3773,4.5453],[3777,4.9054],[3780,5.2109],[3783,5.5545],[3786,5.9362],[3789,6.3562],[3793,6.9758],[3797,7.6632],[3803,8.808],[3806,9.4276],[3809,10.1145],[3811,10.6264],[3813,11.1931],[3815,11.8251],[3817,12.5328],[3818,12.9185],[3819,13.3408],[3820,13.8125],[3821,14.3318],[3822,14.8968],[3823,15.5056],[3824,16.1563],[3825,16.8469],[3826,17.5755],[3828,19.1387],[3830,20.8294],[3832,22.6305],[3835,25.5008],[3848,38.8019],[3849,39.8901],[3850,41.1613],[3851,42.6015],[3852,44.197],[3853,45.9353],[3912,163.7803],[3935,201.0742],[3970,251.7239],[3980,270.0841],[3984,280.0198],[4001,336.7088],[4006,349.5596],[4010,356.563],[4013,359.4235],[4032,360]]
       }
       Scale.X {type=MovieSceneScriptingDoubleChannel, default=50, const [50,433]}
       Scale.Y {type=MovieSceneScriptingDoubleChannel, default=50, const [50,433]}
       Scale.Z {type=MovieSceneScriptingDoubleChannel, default=50, const [50,433]}
      ]
     }]
   }
   {type=MovieSceneSpawnTrack, sections [{type=MovieSceneSpawnSection, channels [None {type=MovieSceneScriptingBoolChannel, default=true}]}]}
  ]
 }
 Tentacle Pistol Canary Creature {
  id=#2
  possessed_class=SkeletalMeshActor
  tracks [
   {type=MovieSceneSpawnTrack, sections [{type=MovieSceneSpawnSection, channels [None {type=MovieSceneScriptingBoolChannel, default=true}]}]}
   {
    type=MovieScene3DTransformTrack
    sections [{
      type=MovieScene3DTransformSection
      channels [
       Location.X {type=MovieSceneScriptingDoubleChannel, default=0, const [14287.208,1]}
       Location.Y {type=MovieSceneScriptingDoubleChannel, default=0, const [70618.9219,1]}
       Location.Z {type=MovieSceneScriptingDoubleChannel, default=0, const [3947.8015,1]}
       Rotation.X {type=MovieSceneScriptingDoubleChannel, default=0, const [0,1]}
       Rotation.Y {type=MovieSceneScriptingDoubleChannel, default=0, const [0,1]}
       Rotation.Z {type=MovieSceneScriptingDoubleChannel, default=0, const [0,1]}
       Scale.X {type=MovieSceneScriptingDoubleChannel, default=1, const [1,1]}
       Scale.Y {type=MovieSceneScriptingDoubleChannel, default=1, const [1,1]}
       Scale.Z {type=MovieSceneScriptingDoubleChannel, default=1, const [1,1]}
      ]
     }]
   }
   {type=MovieSceneSkeletalAnimationTrack, sections [{type=MovieSceneSkeletalAnimationSection, end_frame=4051, start_frame=3600}]}
   {
    type=MovieSceneControlRigParameterTrack
    display_name=Tentacle_PistolCanary_Creature_CtrlRig
    sections [{
      type=MovieSceneControlRigParameterSection
      channels [
       control rig options.Location.X {type=MovieSceneScriptingFloatChannel, default=0}
       control rig options.Location.Y {type=MovieSceneScriptingFloatChannel, default=0}
       control rig options.Location.Z {type=MovieSceneScriptingFloatChannel, default=0}
       control rig options.Rotation.X {type=MovieSceneScriptingFloatChannel, default=0}
       control rig options.Rotation.Y {type=MovieSceneScriptingFloatChannel, default=0}
       control rig options.Rotation.Z {type=MovieSceneScriptingFloatChannel, default=0}
       control rig options.Scale.X {type=MovieSceneScriptingFloatChannel, default=1}
       control rig options.Scale.Y {type=MovieSceneScriptingFloatChannel, default=1}
       control rig options.Scale.Z {type=MovieSceneScriptingFloatChannel, default=1}
       Weight {type=MovieSceneScriptingFloatChannel, default=1}
      ]
     }]
   }
  ]
 }
 _2_Position_CTRL {
  id=#3
  possessed_class=FortStaticMeshActor
  tracks [
   {type=MovieSceneSpawnTrack, sections [{type=MovieSceneSpawnSection, channels [None {type=MovieSceneScriptingBoolChannel, default=true}]}]}
   {
    type=MovieScene3DTransformTrack
    sections [{
      type=MovieScene3DTransformSection
      channels [
       Weight {type=MovieSceneScriptingFloatChannel, default=1}
       Location.X {
        type=MovieSceneScriptingDoubleChannel
        default=14354.7308
        thinned_from=517
        keys [[3600,14250.773],[3606,14252.8154],[3615,14621.4431],[3620,14610.878],[3624,14445.353],[3625,14326.7337],[3626,14114.6204],[3628,13616.5378],[3629,13429.8913],[3630,13342.3532],[3631,13337.0866],[3635,13500.6408],[3650,14466.0161],[3651,14563.4795],[3652,14709.6377],[3655,15300.1142],[3657,15810.4325],[3658,16141.4716],[3660,16913.8811],[3661,17251.6809],[3662,17485.973],[3663,17606.1535],[3664,17666.0289],[3665,17644.151],[3667,17537.5223],[3688,16763.1943],[3696,16340.4822],[3748,14946.5499],[3776,14329.241],[3795,13385.1556],[3817,11972.4624],[3842,10856.8675],[3852,10301.1033],[3867,9720.6575],[3875,9601.7226],[3883,9644.464],[3895,10012.7986],[3901,10322.1268],[3907,10763.7824],[3920,12169.0592],[3936,14100.1174],[3949,15800.8119],[3963,17134.7887],[3976,17923.2743],[3989,18220.0643],[4012,17785.5887],[4024,17165.0884],[4043,15918.8584],[4057,15409.5948],[4078,15346.4338],[4096,15248.6273],[4120,16318.4564],[4122,16326.0952],[4127,16139.2374],[4158,14333.1734],[4161,14255.185],[4182,14250.7725]]
       }
       Location.Y {
        type=MovieSceneScriptingDoubleChannel
        default=69532.1653
        thinned_from=583
        keys [[3600,70457.7959],[3614,70258.4741],[3621,69036.6333],[3623,68541.5649],[3625,68236.1198],[3628,66764.1285],[3629,66366.7325],[3630,66122.9306],[3631,66024.9787],[3650,67627.0071],[3654,67349.7241],[3655,67434.5612],[3656,67670.8973],[3657,68156.6773],[3658,68929.1422],[3661,72034.5269],[3662,72869.8206],[3663,73419.6912],[3664,73681.2217],[3666,73595.9054],[3671,72948.2385],[3852,72176.9093],[3930,66104.4447],[4064,74856.5525],[4146,70336.5254],[4182,70457.7961]]
       }
       Location.Z {
        type=MovieSceneScriptingDoubleChannel
        default=13795.9145
        thinned_from=583
        keys [[3600,15965.2362],[3614,15845.4065],[3620,15987.9955],[3621,15972.772],[3622,15918.6294],[3624,15768.1134],[3625,15753.7294],[3626,15698.0505],[3627,15501.5856],[3628,15121.8624],[3630,14081.4808],[3631,13654.8586],[3632,13487.7379],[3635,13400.4806],[3637,13418.6851],[3639,13521.7936],[3645,14079.3655],[3648,14278.1937],[3651,14576.4159],[3652,14700.2551],[3653,14877.1226],[3654,15095.3645],[3655,15352.0728],[3656,15645.7979],[3658,16397.7038],[3659,16590.6358],[3660,16484.1881],[3661,16015.6343],[3662,15259.0074],[3663,14366.1161],[3664,13387.6921],[3665,12566.2428],[3666,12028.6977],[3667,11728.838],[3668,11633.5887],[3673,11498.5081],[3684,11359.9166],[3695,10933.8946],[3719,10453.099],[3757,10847.3199],[3818,12862.9961],[3879,13289.1351],[3930,12963.052],[4084,13743.3061],[4121,14313.6976],[4126,14444.5214],[4134,14484.947],[4139,14680.9216],[4155,15778.2435],[4160,15951.1726],[4182,15965.236]]
       }
       Rotation.X {
        type=MovieSceneScriptingDoubleChannel
        default=0
        thinned_from=583
        keys [[3600,0.0301],[3601,0.0238],[3602,0.0447],[3603,0.129],[3604,0.2722],[3605,0.4695],[3606,0.6005],[3607,0.5759],[3608,0.4367],[3609,0.2239],[3610,-0.0218],[3611,-0.2788],[3612,-0.5436],[3613,-0.5213],[3614,-0.0086],[3615,0.8774],[3616,2.0315],[3617,3.3534],[3619,5.9541],[3620,7.3511],[3621,8.9126],[3622,10.7836],[3623,13.1537],[3624,16.4698],[3625,27.1927],[3626,92.1886],[3627,156.6798],[3628,167.9043],[3629,171.5925],[3630,173.5315],[3631,174.5533],[3642,179.7729],[3643,-179.8057],[3646,-178.0266],[3650,-173.8565],[3651,-171.8658],[3652,-167.3055],[3653,-160.7594],[3654,-154.8213],[3655,-149.7588],[3656,-143.9649],[3657,-136.7003],[3658,-119.3338],[3659,-74.5783],[3660,-31.943],[3661,-14.0127],[3662,-5.8881],[3663,-1.1083],[3664,1.4394],[3665,3.2297],[3666,4.6853],[3667,4.0842],[3668,2.9087],[3669,2.441],[3670,2.109],[3671,1.9069],[3672,1.8244],[3673,1.8463],[3674,1.8077],[3675,1.7896],[3676,1.8192],[3678,1.9723],[3679,2.0645],[3680,2.1337],[3681,2.2334],[3682,2.3143],[3683,2.4157],[3685,2.6753],[3686,2.7847],[3687,2.8716],[3688,2.9352],[3689,2.9751],[3690,2.9946],[3691,2.987],[3692,2.9668],[3693,2.9528],[3694,2.9759],[3695,3.0574],[3696,3.1973],[3697,3.3956],[3698,3.6152],[3699,3.8515],[3700,4.1027],[3701,4.3683],[3702,4.6478],[3703,4.94],[3704,5.2441],[3705,5.559],[3706,5.8339],[3707,6.042],[3708,6.1508],[3709,6.1891],[3710,6.1995],[3711,6.2664],[3712,6.4363],[3713,6.6891],[3714,7.0053],[3715,7.3877],[3716,7.8165],[3717,8.2805],[3719,9.248],[3720,9.6994],[3721,10.1051],[3723,10.8518],[3729,12.8865],[3732,13.8022],[3735,14.6309],[3736,14.8795],[3737,15.0469],[3739,15.2764],[3740,15.4477],[3742,15.889],[3743,16.0581],[3744,16.1639],[3745,16.1663],[3746,16.0674],[3747,15.8383],[3748,15.4107],[3749,14.8547],[3750,14.2488],[3751,13.7263],[3752,13.3079],[3753,12.9697],[3755,12.402],[3757,11.7463],[3759,11.0136],[3762,9.8124],[3764,8.999],[3765,8.6156],[3766,8.2525],[3769,7.2621],[3772,6.3319],[3774,5.752],[3776,5.211],[3778,4.7074],[3779,4.4706],[3780,4.2449],[3781,4.0291],[3782,3.8236],[3783,3.6292],[3784,3.4457],[3785,3.2725],[3786,3.1102],[3787,2.9584],[3788,2.8178],[3789,2.6878],[3790,2.5684],[3791,2.4599],[3792,2.3619],[3793,2.2749],[3794,2.1983],[3795,2.1322],[3796,2.0772],[3797,2.0323],[3798,1.9974],[3799,1.9737],[3800,1.9585],[3802,1.9508],[3803,1.9397],[3804,1.9026],[3805,1.8587],[3806,1.8103],[3808,1.6862],[3809,1.6332],[3810,1.5738],[3811,1.5207],[3812,1.4576],[3813,1.3885],[3814,1.3132],[3815,1.2336],[3816,1.1492],[3818,0.9717],[3819,0.8799],[3820,0.7831],[3821,0.6782],[3822,0.568],[3823,0.4511],[3824,0.3321],[3826,0.0877],[3827,-0.0359],[3828,-0.157],[3829,-0.2739],[3830,-0.3836],[3831,-0.4887],[3832,-0.5897],[3833,-0.6828],[3834,-0.7684],[3835,-0.8449],[3836,-0.916],[3837,-0.9773],[3838,-1.0228],[3839,-1.0467],[3840,-1.0527],[3841,-1.0386],[3842,-1.0023],[3843,-0.9446],[3844,-0.8666],[3845,-0.7677],[3846,-0.6638],[3847,-0.5663],[3848,-0.483],[3849,-0.4135],[3850,-0.3548],[3851,-0.2911],[3852,-0.214],[3853,-0.1286],[3854,-0.0424],[3855,0.0573],[3856,0.1564],[3857,0.253],[3859,0.4418],[3860,0.5654],[3861,0.7121],[3862,0.883],[3863,1.0797],[3864,1.2914],[3865,1.5215],[3866,1.7816],[3867,2.0463],[3869,2.616],[3874,4.0979],[3875,4.387],[3876,4.6655],[3877,4.9296],[3878,5.174],[3879,5.3899],[3880,5.5779],[3882,5.8989],[3886,6.5052],[3887,6.6712],[3888,6.8507],[3889,7.0454],[3890,7.2587],[3891,7.4947],[3892,7.7541],[3893,8.0295],[3897,9.2052],[3899,9.7584],[3900,10.0138],[3901,10.2465],[3902,10.4575],[3903,10.643],[3904,10.7978],[3905,10.9162],[3907,11.0833],[3908,11.2126],[3909,11.4016],[3910,11.6314],[3912,12.1763],[3915,13.0505],[3916,13.315],[3918,13.7668],[3923,14.7028],[3927,15.5895],[3929,16.0705],[3931,16.4734],[3958,20.1866],[3964,20.7571],[3970,21.0509],[3981,21.0913],[3988,20.7733],[3996,20.2703],[4001,20.2069],[4006,20.227],[4009,20.0977],[4019,19.3177],[4035,18.7533],[4038,18.5132],[4041,18.1302],[4044,17.6132],[4047,16.9691],[4050,16.1844],[4053,15.2732],[4057,13.9127],[4062,12.0897],[4067,10.4137],[4071,9.1565],[4077,7.2915],[4079,6.641],[4081,5.9512],[4082,5.5845],[4083,5.1989],[4084,4.7909],[4085,4.3565],[4087,3.5163],[4088,3.0067],[4089,2.4626],[4090,1.8972],[4091,1.3412],[4092,0.8204],[4093,0.3411],[4094,-0.0911],[4095,-0.4709],[4096,-0.794],[4097,-0.9306],[4098,-1.0387],[4099,-1.123],[4100,-1.1858],[4101,-1.2274],[4102,-1.2341],[4104,-1.0557],[4105,-1.0396],[4106,-1.1015],[4107,-1.2473],[4108,-1.4624],[4109,-1.7495],[4110,-2.099],[4111,-2.4809],[4112,-2.9138],[4113,-3.4133],[4114,-4.0799],[4115,-4.8461],[4116,-5.6592],[4119,-7.9359],[4120,-8.7266],[4121,-9.8871],[4122,-11.0972],[4123,-12.0835],[4124,-12.4701],[4125,-12.5127],[4126,-12.2451],[4127,-11.6339],[4128,-11.7492],[4129,-12.1381],[4130,-12.3315],[4131,-11.7518],[4133,-10.3772],[4134,-9.8909],[4135,-9.7564],[4136,-9.9776],[4137,-9.2396],[4138,-8.0769],[4139,-7.1429],[4140,-7.3028],[4141,-7.2693],[4142,-6.8114],[4143,-6.1656],[4144,-5.8424],[4145,-5.6433],[4146,-5.5017],[4148,-5.3146],[4149,-5.1155],[4151,-4.6246],[4152,-4.2444],[4153,-3.6872],[4155,-2.275],[4156,-1.6721],[4157,-1.1519],[4158,-0.72],[4159,-0.3806],[4160,-0.1339],[4161,-0.011],[4162,0.0301],[4182,0.0301]]
       }
       Rotation.Y {
        type=MovieSceneScriptingDoubleChannel
        default=0
        thinned_from=583
        keys [[3600,40.9752],[3603,40.9896],[3608,40.6915],[3610,40.7094],[3612,40.9061],[3613,41.2117],[3614,42.1862],[3615,44.0025],[3616,46.3649],[3618,51.5451],[3619,54.4975],[3620,58.0304],[3621,61.813],[3623,70.4549],[3624,73.7738],[3625,78.8228],[3626,84.4926],[3627,75.824],[3628,62.5599],[3629,49.5452],[3630,38.3424],[3631,30.1709],[3632,26.6233],[3633,25.7411],[3634,24.7763],[3635,24.0556],[3636,23.5798],[3637,23.3915],[3638,23.5011],[3639,23.8672],[3640,24.4487],[3641,25.2042],[3642,26.1011],[3644,28.1158],[3645,29.0428],[3646,29.78],[3647,30.3639],[3648,31.2022],[3649,32.3186],[3650,33.3504],[3651,34.5188],[3652,36.3637],[3653,39.045],[3654,42.4834],[3655,46.7333],[3656,51.809],[3657,59.4122],[3658,69.1684],[3659,73.8315],[3660,65.6271],[3661,50.4896],[3662,34.4221],[3663,19.477],[3664,4.871],[3665,-6.6865],[3666,-14.0313],[3667,-18.3335],[3668,-19.6742],[3672,-20.4833],[3673,-20.7195],[3676,-20.9417],[3679,-21.2289],[3683,-21.5287],[3684,-21.705],[3685,-22.007],[3686,-22.4481],[3688,-23.5111],[3690,-24.6652],[3691,-25.1826],[3692,-25.6235],[3694,-26.3369],[3695,-26.6153],[3696,-26.8267],[3702,-27.806],[3708,-29.1921],[3710,-29.7162],[3712,-30.0821],[3724,-31.3367],[3733,-31.9355],[3736,-32.3654],[3739,-33.0444],[3751,-36.0226],[3753,-36.326],[3755,-36.4731],[3758,-36.4205],[3761,-36.0565],[3764,-35.4049],[3767,-34.5017],[3771,-32.9868],[3778,-29.7688],[3789,-24.3504],[3793,-22.5772],[3796,-21.3955],[3799,-20.3751],[3802,-19.5395],[3809,-17.8878],[3816,-16.4848],[3821,-15.669],[3824,-15.2826],[3828,-14.9351],[3840,-14.2002],[3842,-14.139],[3844,-14.1372],[3850,-14.3498],[3851,-14.3523],[3852,-14.3152],[3853,-14.243],[3856,-13.9516],[3859,-13.7838],[3861,-13.7719],[3863,-13.841],[3866,-14.0507],[3868,-14.2186],[3870,-14.3146],[3876,-14.3827],[3887,-15.0018],[3890,-15.0829],[3892,-15.0684],[3895,-14.9628],[3898,-14.9552],[3905,-15.1036],[3908,-15.0825],[3910,-15.1542],[3918,-15.7936],[3920,-15.9075],[3922,-15.9483],[3927,-15.883],[3929,-15.8718],[3930,-15.8242],[3934,-15.3887],[3935,-15.306],[3936,-15.2634],[3944,-15.3706],[3950,-15.2753],[3959,-14.8575],[3965,-14.7629],[3974,-14.8912],[3994,-16.0359],[4005,-16.8026],[4020,-17.5204],[4024,-17.5729],[4027,-17.476],[4032,-17.1965],[4042,-16.9789],[4066,-15.8891],[4072,-15.3917],[4076,-14.8085],[4078,-14.4625],[4080,-14.0474],[4082,-13.5392],[4083,-13.244],[4084,-12.9189],[4086,-12.2024],[4087,-11.8141],[4089,-10.9396],[4096,-8.09],[4099,-7.1057],[4101,-6.414],[4102,-6.0285],[4103,-5.4468],[4104,-4.8516],[4105,-4.3808],[4106,-3.9378],[4107,-3.4662],[4108,-3.0164],[4109,-2.6461],[4110,-2.3481],[4111,-2.1107],[4112,-1.8508],[4113,-1.564],[4114,-1.1713],[4115,-0.7161],[4116,-0.2063],[4117,0.3127],[4118,0.8416],[4119,1.349],[4120,1.842],[4121,2.4696],[4122,3.1817],[4123,3.925],[4124,4.5919],[4125,5.1274],[4126,5.5281],[4127,5.8127],[4128,6.1966],[4129,6.6286],[4130,7.0947],[4131,7.4996],[4132,7.9269],[4133,8.5286],[4134,9.2544],[4135,10.2154],[4136,11.2869],[4137,12.4856],[4138,13.7863],[4139,15.2083],[4140,16.7338],[4143,21.4954],[4144,23.0251],[4145,24.4736],[4146,25.8082],[4152,33.4707],[4154,35.8725],[4155,36.9654],[4156,37.9639],[4157,38.8516],[4158,39.5933],[4159,40.2043],[4160,40.6577],[4161,40.8957],[4182,40.9752]]
       }
       Rotation.Z {
        type=MovieSceneScriptingDoubleChannel
        default=0
        thinned_from=583
        keys [[3600,90.0789],[3612,90.4142],[3614,89.9614],[3618,87.8942],[3620,86.4647],[3621,85.9763],[3622,85.885],[3623,86.5548],[3624,88.6249],[3625,98.526],[3626,162.832],[3627,-133.3304],[3628,-122.5352],[3629,-119.0657],[3630,-117.2658],[3631,-116.2095],[3632,-116.0396],[3635,-116.9906],[3645,-118.3935],[3648,-118.0168],[3650,-117.0263],[3651,-115.5231],[3652,-112.5662],[3654,-104.602],[3655,-100.3326],[3656,-95.0682],[3657,-87.5799],[3658,-69.2661],[3659,-23.4711],[3660,19.366],[3661,36.2829],[3662,42.5344],[3663,44.7018],[3664,44.4454],[3665,43.0754],[3666,41.9066],[3667,41.0514],[3668,41.0983],[3671,43.1568],[3673,44.3457],[3683,48.6977],[3684,49.2137],[3685,49.8427],[3687,51.517],[3693,57.7533],[3694,58.7053],[3695,59.5297],[3696,60.2241],[3704,64.1782],[3711,67.1705],[3724,75.5105],[3736,82.3966],[3740,84.9412],[3835,134.7073],[3856,153.94],[3876,179.134],[3877,-179.4534],[3893,-155.4811],[3900,-145.3841],[3926,-97.7394],[3933,-85.7737],[3946,-61.5905],[3953,-49.4366],[3974,-14.9145],[3976,-11.6912],[3978,-8.5146],[3980,-5.3751],[3982,-2.2734],[3983,-0.731],[3984,0.8038],[3985,2.3315],[3987,5.3745],[3989,8.3851],[3990,9.8656],[3991,11.3246],[3992,12.7597],[3993,14.162],[3994,15.527],[3995,16.8551],[3996,18.1396],[3997,19.3686],[3998,20.5503],[4023,49.0689],[4031,57.3492],[4045,69.7037],[4053,75.8237],[4059,79.4962],[4065,82.1511],[4071,83.7487],[4081,85.3359],[4085,85.3151],[4088,86.261],[4096,91.3223],[4100,89.9287],[4103,88.2911],[4113,81.1249],[4116,79.6569],[4125,76.4928],[4127,76.2801],[4132,77.3006],[4142,81.9076],[4151,84.9171],[4152,85.3614],[4156,88.2858],[4158,89.3119],[4160,89.9188],[4182,90.0789]]
       }
       Scale.X {type=MovieSceneScriptingDoubleChannel, default=30, const [30,583]}
       Scale.Y {type=MovieSceneScriptingDoubleChannel, default=30, const [30,583]}
       Scale.Z {type=MovieSceneScriptingDoubleChannel, default=30, const [30,583]}
      ]
     }]
   }
  ]
 }
]
display_rate {denominator=1, numerator=30}
tick_resolution {denominator=1, numerator=24000}
tracks []
